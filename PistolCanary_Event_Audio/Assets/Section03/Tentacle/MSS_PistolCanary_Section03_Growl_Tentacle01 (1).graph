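# Stereo MetaSound source for the PistolCanary Section 03 tentacle growl, playing intermittent random energy one-shots with large point-of-interest attenuation.
# After a random start delay, a repeat trigger picks waves, deriving its period from wave duration plus a random gap.
# A fade module and master volume scale both channels, with fade completion stopping the player.
# OnFinished follows playback end by that period, reset by each new play.
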MetaSoundSource /PistolCanary_Event_Audio/Assets/Section03/Tentacle/MSS_PistolCanary_Section03_Growl_Tentacle01
inputs [
 FadeOut {type=Trigger}
 Master Volume {type=Float}
 Sounds {type=WaveAsset:Array}
 UE.Source.OnPlay {type=Trigger}
 UE.Source.StartTime.StartTime {type=Time}
 Wave Asset {type=WaveAsset}
]
interfaces [UE.OutputFormat.Stereo:1.0, UE.Source.OneShot:1.0, UE.Source.StartTime:1.0, UE.Source:1.1]
outputs [
 "UE.OutputFormat.Stereo.Audio:0" {type=Audio}
 "UE.OutputFormat.Stereo.Audio:1" {type=Audio}
 UE.Source.OneShot.OnFinished {type=Trigger}
]
pages [Default {
  input_defaults {
   FadeOut=null
   Master Volume=1
   UE.Source.OnPlay=null
   UE.Source.StartTime.StartTime=0
   Wave Asset=null
   Sounds [
    /PistolCanary_Event_Audio/Assets/Section03/wav/PistolCanary_Sec_03_TentacleTip_01_Energy_Loop_OS_01
    /PistolCanary_Event_Audio/Assets/Section03/wav/PistolCanary_Sec_03_TentacleTip_01_Energy_Loop_OS_05
   ]
  }
  nodes [
   n1 = &MSP_Fade(In=^FadeOut) {also {defaults {bInvertOutput="<MetaSoundBuilderResult.SUCCEEDED: 0>"}}}
   n2 = UE.RandomFloat(Next=^UE.Source.OnPlay) {also {defaults {Max=3, Min=1}}}
   TriggerDelay_1 = UE.Trigger Delay(Delay Time=UE.ConversionFloatToTime(In=@n2.Value).Out, In=@n2.On Next)
   n3 = VariableDeferredAccessor.Time(Variable=InitVariable.Time().Variable)
   ArrayRandomGet_0 = Array.Random Get:WaveAsset:Array(In Array=^Sounds, Next=UE.TriggerRepeat(Start=@TriggerDelay_1.Out, Period=@n3.Value).RepeatOut) {also {defaults {No Repeats=-1}}}
   n4 = UE.RandomFloat(Next=@ArrayRandomGet_0.On Next) {also {defaults {Max=5, Min=3}}}
   WavePlayer_0 = UE.Wave Player:Stereo(Start Time=^UE.Source.StartTime.StartTime, Wave Asset=@ArrayRandomGet_0.Value, Stop=@n1.On Done, Play=@n4.On Next)
   Multiply_0 = UE.Multiply:Audio by Float(
    AdditionalOperands=^Master Volume
    PrimaryOperand=UE.Multiply:Audio by Float(AdditionalOperands=@n1.Value Out, PrimaryOperand=@WavePlayer_0.Out Right).Out
   )
   Multiply_2 = UE.Multiply:Audio by Float(
    AdditionalOperands=^Master Volume
    PrimaryOperand=UE.Multiply:Audio by Float(AdditionalOperands=@n1.Value Out, PrimaryOperand=@WavePlayer_0.Out Left).Out
   )
   GetWaveDuration_0 = "UE.Get Wave Duration: "(Wave=@ArrayRandomGet_0.Value)
   n5 = UE.Add:Time(PrimaryOperand=UE.ConversionFloatToTime(In=@n4.Value).Out, AdditionalOperands=@GetWaveDuration_0.Duration)
   TriggerDelay_0 = UE.Trigger Delay(
    Delay Time=VariableAccessor.Time(Variable=VariableMutator.Time(Value=@n5.Out, Variable=@n3.Variable).Variable).Value
    In=@WavePlayer_0.On Finished
    Reset=@WavePlayer_0.On Play
   )
  ]
  links [
   {dst="@UE.OutputFormat.Stereo.Audio:1", src=Multiply_0.Out}
   {dst="@UE.OutputFormat.Stereo.Audio:0", src=Multiply_2.Out}
   {dst="@UE.Source.OneShot.OnFinished", src=TriggerDelay_0.Out}
  ]
 }]
properties {
 assetImportData=&self:AssetImportData
 attenuationSettings=/PistolCanary_Audio/Assets/Settings/ATT_PistolCanary_POI_Large
 maxDistance=40000
 numChannels=2
 outputFormat=Stereo
 sampleRate=48000
 soundClassObject=&Fort_Events
}
