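# Sound cue for the slow-motion tentacle slice in Section 3 of the PistolCanary event sequence.
# A single wave player feeds the root directly with the SliceTentacle slow-motion wave, with no randomization or modulation.
# It routes to the Fort_Events sound class and uses a dedicated tentacle-slice attenuation asset, with a very large max distance.
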SoundCue /PistolCanary_Event_Audio/Assets/Section03/Tentacle/SC_PistolCanary_Section01_Tentacle_Slice
root=WavePlayer_0
nodes [{id=WavePlayer_0, properties {soundWaveAssetPtr=/PistolCanary_Event_Audio/Assets/Section03/wav/PistolCanary_Seq_03_SM_SliceTentacle_SloMo}}]
properties {
 attenuationSettings=/PistolCanary_Audio/Assets/Settings/ATT_PistolCanary_Tentacle_Slice
 maxDistance=2097152
 soundClassObject=&Fort_Events
}
