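# Sound cue that plays the Tentacle 01 spawn sound for the emerge moment in Section 03 of the PistolCanary event.
# A single wave player feeds the output directly, with no randomization or modulation nodes.
# It uses the shared tentacle slice attenuation asset, routes to the Fort_Events sound class and slightly boosts volume.
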SoundCue /PistolCanary_Event_Audio/Assets/Section03/Tentacle/SC_PistolCanary_Section03_Emerge_Tentacle01
root=WavePlayer_1
nodes [{id=WavePlayer_1, properties {soundWaveAssetPtr=/PistolCanary_Event_Audio/Assets/Section03/wav/PistolCanary_Sec_03_Tentacle_01_Spawn}}]
properties {
 attenuationSettings=/PistolCanary_Audio/Assets/Settings/ATT_PistolCanary_Tentacle_Slice
 maxDistance=2097152
 soundClassObject=&Fort_Events
 volumeMultiplier=1.2
}
