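# Sound cue for the spawn sound of the second tentacle emerging in Section 03 of the PistolCanary event.
# A single wave player feeds the Tentacle 02 spawn wave straight to the output, with no randomization or modulation.
# It uses the tentacle-tip attenuation asset with a very large max distance, routes to the Fort_Events sound class, and plays at slightly reduced volume.
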SoundCue /PistolCanary_Event_Audio/Assets/Section03/Tentacle/SC_PistolCanary_Section03_Emerge_Tentacle02
root=WavePlayer_1
nodes [{id=WavePlayer_1, properties {soundWaveAssetPtr=/PistolCanary_Event_Audio/Assets/Section03/wav/PistolCanary_Sec_03_Tentacle_02_Spawn}}]
properties {
 attenuationSettings=/PistolCanary_Audio/Assets/Settings/ATT_PistolCanary_Tentacle_Tip
 maxDistance=100000
 soundClassObject=&Fort_Events
 volumeMultiplier=0.8
}
