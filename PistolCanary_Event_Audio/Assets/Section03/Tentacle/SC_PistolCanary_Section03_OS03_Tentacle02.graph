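# Sound cue for the second tentacle-tip energy loop in Section 03 (OS03) of the PistolCanary event.
# A single wave player feeds the loop wave straight to the output, with no mixing or randomization.
# It uses the shared tentacle-tip attenuation settings with a very long max distance and is routed through the Fort_Events sound class.
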SoundCue /PistolCanary_Event_Audio/Assets/Section03/Tentacle/SC_PistolCanary_Section03_OS03_Tentacle02
root=WavePlayer_0
nodes [{
  id=WavePlayer_0
  properties {soundWaveAssetPtr=/PistolCanary_Event_Audio/Assets/Section03/wav/PistolCanary_Sec_03_TentacleTip_02_Energy_Loop_OS_03}
 }]
properties {
 attenuationSettings=/PistolCanary_Audio/Assets/Settings/ATT_PistolCanary_Tentacle_Tip
 maxDistance=100000
 soundClassObject=&Fort_Events
 volumeMultiplier=1
}
