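# Sound cue for the tentacle orb blast in Section 3 of the PistolCanary event, using wave files from the Section 2 folder.
# A Random node with equal weights picks one of two blast variations, positioned by a dedicated tentacle orb blast attenuation asset.
# It plays in the Fort_Events class with overridden concurrency limiting overlap and retriggering.
# A distance-scaled bus send feeds an envelope-follower bus named for side-chain high-pass filtering of the music.
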SoundCue /PistolCanary_Event_Audio/Assets/Section03/Tentacle/SC_PistolCanary_Section03_Orb_Blast
root=Random_0
nodes [
 {id=Random_0, children [WavePlayer_1,WavePlayer_2], properties {weights [1,1]}}
 {id=WavePlayer_1, properties {soundWaveAssetPtr=/PistolCanary_Event_Audio/Assets/Section02/wav/PistolCanary_Tentacle_Attk_6000_Orb_Blast_01}}
 {id=WavePlayer_2, properties {soundWaveAssetPtr=/PistolCanary_Event_Audio/Assets/Section02/wav/PistolCanary_Tentacle_Attk_6000_Orb_Blast_02}}
]
properties {
 attenuationSettings=/PistolCanary_Audio/Assets/Settings/ATT_PistolCanary_Tentacle_Orb_Blast
 bOverrideConcurrency=true
 maxDistance=35500
 soundClassObject=&Fort_Events
 volumeMultiplier=0.9
 busSends [{
   audioBus=/Game/Sounds/Modulation/EnvelopeFollowers/HPF/AB_SideChain_FortMusic_HPF
   maxSendDistance=20000
   maxSendLevel=0.5
   minSendDistance=1000
   minSendLevel=0
   sendLevel=1
   soundSourceBus=null
   sourceBusSendLevelControlMethod=Linear
  }]
 concurrencyOverrides {maxCount=2, retriggerTime=0.3}
}
