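# Sound cue for a tentacle death in Section 03 of the PistolCanary event.
# A mixer layers the Section 03 death wave, played immediately at full volume, with a quieter tentacle attack death wave reused from Section 02 that follows after a fixed delay.
# Output uses the shared tentacle tip attenuation settings and the Fort_Events sound class.
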SoundCue /PistolCanary_Event_Audio/Assets/Section03/Tentacle/SC_PistolCanary_Section03_Tentacle_Death_01
root=Mixer_0
nodes [
 {id=Delay_0, children [WavePlayer_1], properties {delayMax=2.4, delayMin=2.4}}
 {id=Mixer_0, children [WavePlayer_0,Delay_0], properties {inputVolume [1,0.7]}}
 {id=WavePlayer_0, properties {soundWaveAssetPtr=/PistolCanary_Event_Audio/Assets/Section03/wav/PistolCanary_Sec_03_Tentacle_Death_05}}
 {id=WavePlayer_1, properties {soundWaveAssetPtr=/PistolCanary_Event_Audio/Assets/Section02/wav/PistolCanary_Tentacle_Attk_1300_Death}}
]
properties {
 attenuationSettings=/PistolCanary_Audio/Assets/Settings/ATT_PistolCanary_Tentacle_Tip
 maxDistance=100000
 soundClassObject=&Fort_Events
 volumeMultiplier=1
}
