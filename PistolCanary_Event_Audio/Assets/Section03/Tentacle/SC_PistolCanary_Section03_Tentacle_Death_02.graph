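# Sound cue for a Section 03 tentacle death in the PistolCanary event audio.
# A mixer layers two waves at equal volume: Death_03 plays immediately and Death_04 follows after a fixed 1.5-second delay.
# The cue uses the shared PistolCanary tentacle-death attenuation, is audible over a long range, and routes to the Fort_Events sound class.
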SoundCue /PistolCanary_Event_Audio/Assets/Section03/Tentacle/SC_PistolCanary_Section03_Tentacle_Death_02
root=Mixer_0
nodes [
 {id=Delay_0, children [WavePlayer_1], properties {delayMax=1.5, delayMin=1.5}}
 {id=Mixer_0, children [WavePlayer_0,Delay_0], properties {inputVolume [1,1]}}
 {id=WavePlayer_0, properties {soundWaveAssetPtr=/PistolCanary_Event_Audio/Assets/Section03/wav/PistolCanary_Sec_03_Tentacle_Death_03}}
 {id=WavePlayer_1, properties {soundWaveAssetPtr=/PistolCanary_Event_Audio/Assets/Section03/wav/PistolCanary_Sec_03_Tentacle_Death_04}}
]
properties {
 attenuationSettings=/PistolCanary_Event_Audio/Assets/Settings/ATT_PistolCanary_Tentacle_Deaths
 maxDistance=35000
 soundClassObject=&Fort_Events
 volumeMultiplier=1
}
