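# Sound cue for a tentacle death sound in Section 03 of the PistolCanary event audio.
# A single wave player feeds the output directly with the Death_05 wave, despite the cue's 03 suffix, with no randomization or mixing.
# The cue uses the shared tentacle-death attenuation asset and the Fort_Events sound class, with slightly raised pitch and reduced volume.
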SoundCue /PistolCanary_Event_Audio/Assets/Section03/Tentacle/SC_PistolCanary_Section03_Tentacle_Death_03
root=WavePlayer_0
nodes [{id=WavePlayer_0, properties {soundWaveAssetPtr=/PistolCanary_Event_Audio/Assets/Section03/wav/PistolCanary_Sec_03_Tentacle_Death_05}}]
properties {
 attenuationSettings=/PistolCanary_Event_Audio/Assets/Settings/ATT_PistolCanary_Tentacle_Deaths
 maxDistance=35000
 pitchMultiplier=1.1
 soundClassObject=&Fort_Events
 volumeMultiplier=0.7
}
